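# Unlit emissive material for the inverted BlueSteel SE vehicle wheel, enabled for skeletal meshes.
# Rim and tire masks multiply a panning geometric texture, with floored and clamped results tinted by EmissiveColor and adjustable intensities.
# The rim mask's green channel blends in thread lights, and vertex color red switches rim to tire; two inner lerps have identical inputs.
# The output passes through shared vehicle damage and de-rez transition functions.
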Material /VehicleCosmetics/Wheels/BlueSteelSE_Inverted/Material/M_Wheel_Bluesteel_Inverted
graph {
 properties {
  bAutomaticallySetUsageInEditor=false
  bUseMaterialAttributes=true
  bUsedWithMeshDeformer=true
  bUsedWithSkeletalMesh=true
  shadingModel=MSM_Unlit
  usedShadingModels=Unlit
 }
 nodes [
  n1 = VectorParameter() {group=Inverted, parameterName=EmissiveColor, defaultValue=<rgba 0.1 0.4 1 1>}
  n2 = Multiply(A=@n1.RGB, B=Constant() {r=3})
  n3 = TextureSampleParameter2D() {
   parameterName=Rim_Emissive
   samplerType=SAMPLERTYPE_LinearColor
   texture=/VehicleCosmetics/Wheels/BlueSteelSE_Inverted/Textures/T_BlueSteelInverted_M
  }
  n4 = ScalarParameter() {defaultValue=1, group=Inverted, parameterName=RimEmissiveGlow, sortPriority=3}
  n5 = TextureCoordinate() {uTiling=0.5}
  n6 = Multiply(
   A=Multiply(A=@n3.R, B=@n4)
   B=TextureSample(UVs=Panner(Coordinate=@n5) {speedX=0.5}) {samplerSource=SSM_Wrap_WorldGroupSettings, texture=/VehicleCosmetics/Wheels/_SharedTextures/FX/Skin_GeometricBM_RGB}.R
  )
  n7 = Multiply(A=@n3.R, B=@n4)
  n8 = TextureSample(UVs=Panner(Coordinate=TextureCoordinate() {uTiling=0.5}) {speedX=0.5}) {samplerSource=SSM_Wrap_WorldGroupSettings, texture=&Skin_GeometricBM_RGB}
  n9 = Clamp(None=Add(A=Floor(None=@n6), B=Multiply(A=@n7, B=@n8.R))) {maxDefault=3, minDefault=0.001}
  n10 = LinearInterpolate(A=@n9, B=@n9, Alpha=VertexColor().R)
  n11 = TextureSampleParameter2D() {
   parameterName=Rim_Emissive
   samplerType=SAMPLERTYPE_LinearColor
   texture=&T_BlueSteelInverted_M
  }
  n12 = Multiply(A=@n11.G, B=@n1.RGB)
  n13 = ScalarParameter() {defaultValue=6, group=Inverted, parameterName=RimThreadLights, sortPriority=4}
  n14 = VectorParameter() {group=Inverted, parameterName=EmissiveColor, defaultValue=<rgba 0.1 0.4 1 1>}
  n15 = ScalarParameter() {defaultValue=6, group=Inverted, parameterName=TireThreadLights, sortPriority=4}
  n16 = TextureSample() {samplerType=SAMPLERTYPE_LinearColor, texture=/VehicleCosmetics/Wheels/BlueSteelSE_Inverted/Textures/T_BlueSteelInverted_Tire_M}
  n17 = ScalarParameter() {defaultValue=1, group=Inverted, parameterName=TireEmissiveGlow, sortPriority=3}
  n18 = TextureCoordinate() {uTiling=0.5}
  n19 = Multiply(
   A=Multiply(A=@n16.R, B=@n17)
   B=TextureSample(UVs=Panner(Coordinate=@n18) {speedX=0.5}) {samplerSource=SSM_Wrap_WorldGroupSettings, texture=&Skin_GeometricBM_RGB}.R
  )
  n20 = Multiply(A=@n16.R, B=@n17)
  n21 = TextureSample(UVs=Panner(Coordinate=TextureCoordinate() {uTiling=0.5}) {speedX=0.5}) {samplerSource=SSM_Wrap_WorldGroupSettings, texture=&Skin_GeometricBM_RGB}
  n22 = Clamp(None=Add(A=Floor(None=@n19), B=Multiply(A=@n20, B=@n21.R))) {maxDefault=3, minDefault=0.001}
  n23 = VertexColor()
  n24 = LinearInterpolate(
   A=LinearInterpolate(A=Multiply(A=@n2, B=@n10), B=Multiply(A=@n12, B=@n13), Alpha=@n11.G)
   B=Multiply(A=Multiply(A=@n14.RGB, B=@n15), B=LinearInterpolate(A=@n22, B=@n22, Alpha=@n23.R))
   Alpha=VertexColor().R
  )
  MaterialFunctionCall(In=MaterialFunctionCall(Material=SetMaterialAttributes(Emissive Color=@n24) {attributeSetTypes [#1]}) {materialFunction=&MF_BaseVehicle_Damage}.Result) {materialFunction=&MF_VehicleDeRez_Transition}
 ]
}
